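# Clear-coat vehicle material for CoolDudeCar, with parameterized paint, chrome and optional decals.
# View- and sun-reactive paint with noise-based metallic speckles blends by mask into chrome with Fresnel rainbow tint.
# A switchable decal layer and textured detail areas sit on top, with masked emissive driven by time of day, lit at night.
# Shading-path switches simplify mobile, then baked normal and MF_GameWideShared finish the output.
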Material /Game/Vehicles/FORT_Vehicles_01/Materials/M_CoolDudeCar
graph {
 properties {bUseMaterialAttributes=true, pixelDepthOffsetMode=PDOM_Legacy, shadingModel=MSM_ClearCoat, usedShadingModels=Clear Coat}
 nodes [
  n1 = VectorParameter() {group=CarPaint, parameterName=CarPaintColor, defaultValue=<rgba 0.085 0.0941 0.1 0>}
  n2 = Saturate(None=DotProduct(A=CameraVectorWS(), B=VertexNormalWS()))
  n3 = DotProduct(A=AtmosphericLightVector(), B=ReflectionVectorWS())
  n4 = Power(Base=@n3, Exp=Constant() {r=25})
  n5 = Multiply(A=@n2, B=@n4)
  n6 = LinearInterpolate(A=Multiply(A=@n1.RGB) {constB=0.25}, B=@n1.RGB, Alpha=Saturate(None=Add(A=@n2, B=@n5)))
  n7 = Constant() {r=10}
  n8 = Constant() {r=0.1}
  n9 = Power(Base=@n3, Exp=Constant() {r=1500})
  n10 = Constant() {r=5}
  n11 = Multiply(A=TextureCoordinate(), B=ScalarParameter() {defaultValue=50, parameterName=Param_2})
  n12 = TextureSample(UVs=@n11) {
   samplerSource=SSM_Wrap_WorldGroupSettings
   samplerType=SAMPLERTYPE_LinearColor
   texture=/Game/Gadgets/Assets/Slow_Field/Textures/T_Slowfield_SparkleNoise_01
  }
  n13 = TextureSample(UVs=Multiply(A=@n11, B=Constant() {r=2})) {
   samplerSource=SSM_Wrap_WorldGroupSettings
   samplerType=SAMPLERTYPE_LinearColor
   texture=&T_Slowfield_SparkleNoise_01
  }
  n14 = Power(Base=Add(A=@n12.R, B=@n13.G), Exp=Constant() {r=25})
  n15 = Constant() {r=555}
  n16 = Add(A=Saturate(None=Multiply(A=@n4, B=@n8)), B=Multiply(A=Multiply(A=@n9, B=@n10), B=Multiply(A=@n14, B=@n15)))
  n17 = OneMinus(None=Fresnel())
  n18 = Multiply(A=Saturate(None=Multiply(A=@n16, B=@n17)), B=ScalarParameter() {defaultValue=1, group=CarPaint, parameterName=Metal Speckles})
  n19 = ScalarParameter() {defaultValue=0.35, group=CarPaint, parameterName=CarPaintRoughness}
  n20 = SetMaterialAttributes!(
   MaterialAttributes
   Base Color=LinearInterpolate(A=@n6, B=Multiply(A=@n6, B=@n7), Alpha=@n18)
   Metallic=@n18
   Specular=@n6
   Roughness=LinearInterpolate(A=@n19, B=ScalarParameter() {defaultValue=0.25, group=CarPaint, parameterName=SpecklesRoughness}, Alpha=@n18)
   Clear Coat=Constant() {r=1}
   Clear Coat Roughness=Constant() {r=0.001}
  ) {attributeSetTypes [
    #1
    #2
    #3
    #4
    #5
    #6
   ]}
  n21 = SetMaterialAttributes!(
   MaterialAttributes
   Base Color=ScalarParameter() {defaultValue=0.8, group=Chrome, parameterName=ChromeTone}
   Metallic=Constant() {r=1}
   Roughness=ScalarParameter() {defaultValue=0.2, group=Chrome, parameterName=ChromeRoughness}
  ) {attributeSetTypes [#1, #2, #4]}
  n22 = GetMaterialAttributes(None=@n21) {attributeGetTypes [#1]}
  n23 = Fresnel(Normal=VertexNormalWS())
  n24 = Multiply(
   A=TextureSample(UVs=OneMinus(None=@n23)) {samplerSource=SSM_Clamp_WorldGroupSettings, texture=/Game/Characters/Fortress_CharacterShader/Textures/RainbowEffect1}.RGB
   B=@n22.Base Color
  )
  n25 = Fresnel(Normal=VertexNormalWS())
  n26 = SetMaterialAttributes!(MaterialAttributes=@n22.MaterialAttributes, Base Color=Add(A=Multiply(A=@n24, B=@n25), B=@n22.Base Color)) {attributeSetTypes [#1]}
  n27 = TextureSampleParameter2D() {
   group=Base
   parameterName=Car M
   samplerSource=SSM_Clamp_WorldGroupSettings
   samplerType=SAMPLERTYPE_Masks
   texture=&/CoolDudeCar_M
  }
  n28 = BlendMaterialAttributes(
   A=ShadingPathSwitch(
    Default=@n20
    Deferred=@n20
    Mobile=SetMaterialAttributes!(MaterialAttributes, Base Color=@n1.RGB, Roughness=@n19) {attributeSetTypes [#1, #4]}
   ) {desc=Migrated from FeatureLevelSwitch}
   B=ShadingPathSwitch(Default=@n26, Deferred=@n26, Mobile=@n21) {desc=Migrated from FeatureLevelSwitch}
   Alpha=@n27.G
  )
  n29 = TextureSampleParameter2D() {
   group=Decal
   parameterName=Decal D
   samplerSource=SSM_Clamp_WorldGroupSettings
   texture=&/CoolDudeCar_Decal_D
  }
  n30 = Multiply(A=VectorParameter() {group=Decal, parameterName=Decal, defaultValue=<rgba 0.03 0.03 0.03 0>}.RGB, B=@n29.RGB)
  n31 = TextureSampleParameter2D() {
   group=Decal
   parameterName=Decal N
   samplerSource=SSM_Clamp_WorldGroupSettings
   samplerType=SAMPLERTYPE_Normal
   texture=&/CoolDudeCar_Decal_N
  }
  n32 = SetMaterialAttributes!(
   MaterialAttributes
   Base Color=@n30
   Specular=Multiply(A=@n30) {constB=2}
   Roughness=Constant() {r=0.3}
   Normal=MaterialFunctionCall(Normal=@n31.RGB, Flatness=Constant() {r=0.75}) {materialFunction=&FlattenNormal}.Result
   Clear Coat=Constant()
  ) {attributeSetTypes [
    #1
    #3
    #4
    #7
    #5
   ]}
  n33 = SetMaterialAttributes!(MaterialAttributes, Base Color=@n30) {attributeSetTypes [#1]}
  n34 = StaticSwitchParameter(
   True=BlendMaterialAttributes(A=@n28, B=ShadingPathSwitch(Default=@n32, Deferred=@n32, Mobile=@n33) {desc=Migrated from FeatureLevelSwitch}, Alpha=@n29.A)
   False=@n28
  ) {parameterName=Decals True}
  n35 = GetMaterialAttributes(None=@n34) {attributeGetTypes [#1]}
  n36 = SetMaterialAttributes(
   MaterialAttributes=@n35.MaterialAttributes
   Base Color=Multiply(
    A=@n35.Base Color
    B=TextureSample() {samplerSource=SSM_Clamp_WorldGroupSettings, texture=&/CoolDudeCar_D}.RGB
   )
  ) {attributeSetTypes [#1]}
  n37 = TextureSampleParameter2D() {group=Base, parameterName=Car D, samplerSource=SSM_Clamp_WorldGroupSettings, texture=&CoolDudeCar_D}
  n38 = TextureSampleParameter2D() {
   group=Base
   parameterName=Car M
   samplerSource=SSM_Clamp_WorldGroupSettings
   samplerType=SAMPLERTYPE_Masks
   texture=&CoolDudeCar_M
  }
  n39 = SetMaterialAttributes!(MaterialAttributes, Base Color=@n37.RGB, Roughness=@n38.B, Specular=@n38.B, Clear Coat=Constant()) {attributeSetTypes [
    #1
    #4
    #3
    #5
   ]}
  n40 = BlendMaterialAttributes(A=ShadingPathSwitch(Default=@n36, Deferred=@n36, Mobile=@n34) {desc=Migrated from FeatureLevelSwitch}, B=@n39, Alpha=@n37.A)
  n41 = Multiply(A=@n37.RGB, B=@n38.A)
  n42 = Constant()
  n43 = MaterialFunctionCall(Color_Morning=@n42, Color_Day=@n42, Color_Evening=@n42, Color_Night=ScalarParameter() {defaultValue=2, group=Base, parameterName=Emissive}) {materialFunction=&Color_Time_of_Day}
  n44 = Fresnel(BaseReflectFractionIn=Constant() {r=0.5}) {baseReflectFraction=-0.1905, exponent=1}
  n45 = SetMaterialAttributes!(MaterialAttributes=@n40, Emissive Color=Multiply(A=Multiply(A=@n41, B=@n43.Result), B=OneMinus(None=@n44))) {attributeSetTypes [#8]}
  n46 = TextureSampleParameter2D() {
   group=Base
   parameterName=Car N
   samplerSource=SSM_Clamp_WorldGroupSettings
   samplerType=SAMPLERTYPE_Normal
   texture=&/CoolDudeCar_N
  }
  MaterialFunctionCall!(In=MaterialFunctionCall!(Material=ShadingPathSwitch(Default=@n45, Deferred=@n45, Mobile=@n40) {desc=Migrated from FeatureLevelSwitch}, Normal=@n46.RGB) {materialFunction=/Engine/Functions/MaterialLayerFunctions/MatLayerBlend_BakedNormal}.Blended Material) {materialFunction=/Game/Athena/WinterArt/Materials/MF_GameWideShared}
 ]
}
