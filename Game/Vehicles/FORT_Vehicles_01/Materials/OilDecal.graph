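# Translucent deferred decal material for oil stains, filed with the vehicle assets.
# Half-tiled UVs shifted by XOffset and YOffset parameters sample the OilStains grayscale mask and OilStainsNormal map.
# The mask feeds HeightLerp as a height texture to give a contrast-adjusted blend.
# The stain normal adds faint XY detail from a swirl noise normal at a fixed mip level, alongside constants and a Metallic parameter.
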Material /Game/Vehicles/FORT_Vehicles_01/Materials/OilDecal
graph {
 properties {
  bTangentSpaceNormal=false
  bUsedWithNanite=true
  blendMode=BLEND_Translucent
  materialDomain=MD_DeferredDecal
  pixelDepthOffsetMode=PDOM_Legacy
  usedShadingModels=Default Lit
 }
 nodes [
  n1 = Add(
   A=TextureCoordinate() {uTiling=0.5, vTiling=0.5}
   B=AppendVector(A=ScalarParameter() {parameterName=XOffset}, B=ScalarParameter() {parameterName=YOffset})
  )
  n2 = TextureSample!(UVs, Tex, Level, Apply View MipBias) {
   constMipValue=2
   mipValueMode=TMVM_MipLevel
   samplerSource=SSM_Wrap_WorldGroupSettings
   samplerType=SAMPLERTYPE_Normal
   texture=/Game/Global/Textures/Default/Noise/T_SwirlNoise_N
  }
  Add(
   A=TextureSample!(UVs=@n1, Tex, Apply View MipBias) {samplerType=SAMPLERTYPE_Normal, texture=&/OilStainsNormal}.RGB
   B=Multiply(A=@n2.RGB, B=Constant3Vector() {constant=<rgba 0.1 0.1 0 0>})
  )
  Constant() {r=0.05}
  Constant() {r=0.5}
  Constant()
  MaterialFunctionCall(
   Transition Phase=Constant() {r=0.2857}
   Height Texture=TextureSample(UVs=@n1) {samplerType=SAMPLERTYPE_Grayscale, texture=&/OilStains}.R
   Contrast=Constant() {r=2.5}
  ) {materialFunction=&HeightLerp}
  ScalarParameter() {defaultValue=0.5, parameterName=Metallic}
 ]
}
